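# Niagara one-shot burst effect for the default prop spawn in the PixieParts emote, combining smoke, sparkles, a flare and a shockwave.
# DefaultPropSpawn_Merged packs smoke, flash and curl-noise sparks into one emitter using spawn-group switches, with platform-dependent burst counts.
# BigSparkles adds flickering glints, Flare a blue sprite and light, and Shockwave expanding blue meshes.
# Time-of-day and front-end scalars adjust brightness and size.
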NiagaraSystem /BRCosmetics/Animation/Game/MainPlayer/Emotes/PixieParts/FX/NS_Emote_PixiePart_DefaultPropSpawn
dependencies {
 usedDataInterfaces [
  &NiagaraDataInterfacePlatformSet
  &NiagaraDataInterfaceCurve
  &NiagaraDataInterfaceArrayFloat
  &NiagaraDataInterfaceColorCurve
 ]
 usedDynamicInputs [&ReturnNormalizedExecIndex, &FNTimeOfDayScalar]
 usedModules [
  &self:SystemSpawnScript.SystemScriptSource.SystemScriptGraph.NiagaraNodeAssignment_0.SetVariables_#1
  &SystemState
 ]
}
emitters {
 BigSparkles {
  emitterUpdateScript [
   &EmitterState()
   &SpawnBurst_Instantaneous(Spawn Count=5, Use Spawn Probability=false, UseLoopCountLimit=false)
  ]
  particleSpawnScript [
   /SetVariables(Particles.VisibilityTag=&RandomRangeInt(Minimum=0, Maximum=1)) {isSetParameters=true}
   &InitializeParticle(
    Color=<rgba 5 3.7043 1.51 1>
    Lifetime=&RandomRangeFloat(Minimum=0.3, Maximum=0.6)
    Sprite Size Mode=::NewEnumerator3
    Uniform Sprite Size=&Multiply_Float(
     A=&SelectFloatFromArray(Direct Array Index=$Particles.VisibilityTag, Float Selection Array=DataInterface() {propertyValues {FloatData [80,64]}})
     B=&RandomRangeFloat(Minimum=0.5, Maximum=1.3)
    )
    UsePositionOffset=true
    Write Material Random=true
   )
   &CameraOffset(Camera Offset Amount=80)
   &ShapeLocation(MaskSpawnGroup=false, Offset=<0 0 30>, Offset Mode=::NewEnumerator0, Override Seed=false, Sphere Radius=50, UseOwnerScale=true)
   &AddVelocity(Velocity=&RandomRangeVector(Minimum=<0 -20 0>, Maximum=<1 -60 1>))
  ]
  particleUpdateScript [
   &ParticleState()
   &ScaleSpriteSize(
    Scale Factor=&Multiply_Vector2D_ByFloat(
     Float=1
     Vector2D=&Vector2DFromFloat(Value=&FloatFromCurve(FloatCurve=DataInterface() {propertyValues {Curve {keys [
           {arriveTangent=9.0592, interpMode=RCIM_Cubic, leaveTangent=9.0592, tangentMode=RCTM_User, time=0, value=0}
           {interpMode=RCIM_Cubic, tangentMode=RCTM_User, time=0.2, value=1}
           {tangentMode=RCTM_User, time=1, value=0}
          ]}}}))
    )
    Scale Sprite Size Mode=::NewEnumerator2
   )
   /SetVariables(
    Particles.Color02:LinearColor=&MultiplyLinearColor(
     Linear Color=$Particles.Color
     Scale Factor=&MakeVector4_FromLinearColor(LinearColor=<rgba 0.9441 0.8079 1 1>)
    )
    Particles.SpriteSize02:Vector2f=&Multiply_Vector2(A=$Particles.SpriteSize, B=&Vector2DFromFloat(Value=1))
   ) {isSetParameters=true}
   &ScaleColor(
    Scale Alpha=&Multiply_Float(
     A=&FloatFromCurve(
      FloatCurve=DataInterface() {propertyValues {Curve {keys [
          {interpMode=RCIM_Cubic, tangentMode=RCTM_User, time=0, value=0}
          {arriveTangent=-2.4058, interpMode=RCIM_Cubic, leaveTangent=-2.4058, tangentMode=RCTM_User, time=0.26, value=0.8421}
          {arriveTangent=-1.1775, interpMode=RCIM_Cubic, leaveTangent=-1.1775, time=0.4613, value=0.0645}
          {arriveTangent=-0.1197, interpMode=RCIM_Cubic, leaveTangent=-0.1197, time=0.5898, value=0.4537}
          <at 1 0 RCIM_Cubic>
         ]}}}
      Scale Curve=&Multiply_Float(A=&RandomRangeFloat(Minimum=0.1, Maximum=0.5), B=$System.TODScalar)
     )
     B=&FloatFromCurve(
      FloatCurve=DataInterface() {propertyValues {Curve {keys [{tangentMode=RCTM_User, time=0, value=0}, {tangentMode=RCTM_User, time=1, value=1}]}}}
      CurveIndex=&Multiply_Float(
       A=&MakeFloatFromVector(Vector=Linked() {linkedVariable {name=Particles.Position, type=NiagaraPosition}}, Channel=::NewEnumerator2)
       B=$Engine.System.Age
      )
     )
    )
    ScaleA=true
    ScaleRGB=true
   )
   &CurlNoiseForce(Bias Noise=false, Noise Strength=100, Pan Noise=false)
   &Drag(Drag=&RandomRangeFloat(Minimum=10, Maximum=12), Use Linear Drag=true, Use Rotational Drag=true)
   &SolveForcesAndVelocity(Clamp Velocity=false, Limit Acceleration=false)
  ]
  rendererClasses [&NiagaraSpriteRendererProperties]
  renderers [
   {rendererClass=&NiagaraSpriteRendererProperties}
   {rendererClass=&NiagaraSpriteRendererProperties, rendererIndex=1}
  ]
 }
 DefaultPropSpawn_Merged {
  emitterUpdateScript [
   &EmitterState()
   &SpawnBurst_Instantaneous(
    Spawn Count=&MakeIntFromFloat(Float=&FN_PickFloatDependOnPlatform(TargetPlatformActive=7, TargetPlatformNoActive=3))
    Use Spawn Probability=false
    UseLoopCountLimit=false
   ) {module=SpawnBurst_Instantaneous001}
   &SpawnBurst_Instantaneous(Spawn Group=1, Use Spawn Probability=false, UseLoopCountLimit=false)
   &SpawnBurst_Instantaneous(
    Spawn Count=&MakeIntFromFloat(Float=&FN_PickFloatDependOnPlatform(TargetPlatformActive=13, TargetPlatformNoActive=6))
    Spawn Group=2
    Use Spawn Probability=false
    UseLoopCountLimit=false
   ) {module=SpawnBurst_Instantaneous002}
  ]
  particleSpawnScript [
   /SetVariables(Particles.VisibilityTag=&SpawnGroup()) {isSetParameters=true}
   &InitializeParticle(
    Color=&MergeHelper_ColorSwitch(
     0=<rgba 0.5255 0.5255 0.5255 1>
     1=<rgba 0.4323 0.4323 0.4323 0.2>
     2=<rgba 5 3.7043 1.51 1>
     3=?
     Index=&SpawnGroup()
     Num=::NewEnumerator1
    )
    Lifetime=&MergeHelper_FloatSwitch(
     0=&UniformRangedFloat(Minimum=0.2, Maximum=0.6)
     1=0.25
     2=&UniformRangedFloat(Minimum=0.3, Maximum=1)
     3=?
     Index=&SpawnGroup()
     Num=::NewEnumerator1
    )
    Position Offset=&MergeHelper_VectorSwitch(0=<0 0 0>, 1=<0 0 0>, 2=<4 0 0>, 3=?, Index=&SpawnGroup(), Num=::NewEnumerator1)
    Sprite Rotation Angle=&MergeHelper_FloatSwitch(
     0=&UniformRangedFloat(Minimum=0, Maximum=1)
     1=0
     2=0
     3=?
     Index=&SpawnGroup()
     Num=::NewEnumerator1
    )
    Sprite Rotation Mode=::NewEnumerator1
    Sprite Size Mode=::NewEnumerator3
    Uniform Sprite Size=&MergeHelper_FloatSwitch(
     0=&UniformRangedFloat(Minimum=25, Maximum=35)
     1=120
     2=&UniformRangedFloat(Minimum=20, Maximum=40)
     3=?
     Index=&SpawnGroup()
     Num=::NewEnumerator1
    )
    UsePositionOffset=true
    Write Material Random=true
   ) {module=InitializeParticle001}
   /SetVariables(Particles.InitSpriteRotRate=&MergeHelper_FloatSwitch(
     0=&UniformRangedFloat(Minimum=-1, Maximum=1)
     1=0
     2=0
     3=?
     Index=&SpawnGroup()
     Num=::NewEnumerator1
    )) {isSetParameters=true}
   &AccelerationForce(Acceleration=&MergeHelper_VectorSwitch(
     0=&UniformRangedVector(Minimum=<0 0 -10>, Maximum=<0 0 -20>)
     1=<0 0 0>
     2=<0 0 0>
     3=?
     Index=&SpawnGroup()
     Num=::NewEnumerator1
    ))
   &ShapeLocation(
    Hemisphere Distribution=<1 0.5>
    MaskSpawnGroup=false
    Non Uniform Scale=<1 1 0.1>
    Override Seed=false
    Rotation Coordinate Space=::World
    Sphere Radius=&MergeHelper_FloatSwitch(0=10, 1=0, 2=0, 3=?, Index=&SpawnGroup(), Num=::NewEnumerator1)
    UseOwnerScale=true
   )
   &CameraOffset(Camera Offset Amount=&MergeHelper_FloatSwitch(
     0=&RandomRangeFloat(Minimum=0, Maximum=15)
     1=10
     2=0
     3=?
     Index=&SpawnGroup()
     Num=::NewEnumerator1
    ))
   &AddVelocity(
    Module.Use Falloff Distance=false
    Override Seed=false
    Velocity Mode=::NewEnumerator1
    Velocity Speed=&MergeHelper_FloatSwitch(
     0=&RandomRangeFloat(Minimum=100, Maximum=130)
     1=0
     2=0
     3=?
     Index=&SpawnGroup()
     Num=::NewEnumerator1
    )
   ) {module=AddVelocity001}
   &AddVelocity(
    Rotation Coordinate Space=::World
    Velocity=&MergeHelper_VectorSwitch(
     0=&RandomRangeVector(Minimum=<0 0 0>, Maximum=<0 0 20>)
     1=<0 0 0>
     2=&UniformRangedVector(Minimum=<-200 -200 -200>, Maximum=<200 200 375>)
     3=?
     Index=&SpawnGroup()
     Num=::NewEnumerator1
    )
   )
   &ApplyInitialForces()
   /SetVariables(Particles.DynamicMaterialParameter_Air:Vector4f=&MakeVector4(X=2, Y=3, Z=0, W=0)) {isSetParameters=true}
  ]
  particleUpdateScript [
   &ParticleState()
   &SpriteRotationRate(Rotation Rate=&Multiply_Float(
     A=$Particles.InitSpriteRotRate
     B=&FloatFromCurve(
      FloatCurve=DataInterface() {propertyValues {Curve {keys [
          {arriveTangent=-1.9214, interpMode=RCIM_Cubic, leaveTangent=-1.9214, tangentMode=RCTM_User, time=-0.0046, value=1.0113}
          {arriveTangent=-0.2496, interpMode=RCIM_Cubic, leaveTangent=-0.2496, tangentMode=RCTM_User, time=0.9896, value=0.16}
         ]}}}
      CurveIndex=$Emitter.Age
     )
    ))
   &GravityForce(Gravity=&MergeHelper_VectorSwitch(
     0=&Multiply_VectorByFloat(
      Vector=&RandomRangeVector(Minimum=<0 0 0>, Maximum=<0 0 0>)
      Float=&FloatFromCurve(FloatCurve=DataInterface() {propertyValues {Curve {keys [<at 0.0907 0 RCIM_Cubic>, {arriveTangent=3.1758, interpMode=RCIM_Cubic, leaveTangent=3.1758, tangentMode=RCTM_User, time=0.3, value=1}]}}})
     )
     1=<0 0 0>
     2=<0 0 200>
     3=?
     Index=&SpawnGroup()
     Num=::NewEnumerator1
    ))
   /Niagara/Modules/Update/Forces/CurlNoiseForce(
    Noise Frequency:NiagaraFloat=&MaskFloatBySpawnGroup(Masked Float=0.02, Passthrough Float=0, Spawn Group Mask=2)
    Noise Quality / Cost=::NewEnumerator0
    Noise Strength=&MaskFloatBySpawnGroup(Masked Float=500, Passthrough Float=0, Spawn Group Mask=2)
    Pan Noise=false
   )
   &Drag(
    Drag=&MergeHelper_FloatSwitch(0=3, 1=0, 2=5, 3=?, Index=&SpawnGroup(), Num=::NewEnumerator1)
    Use Linear Drag=true
    Use Rotational Drag=true
   )
   &SolveForcesAndVelocity(Clamp Velocity=false, Limit Acceleration=false)
   &ScaleSpriteSize(
    Scale Factor=&MergeHelper_Vector2Switch(
     0=&Vector2DFromFloat(Value=&FloatFromCurve(FloatCurve=DataInterface() {propertyValues {Curve {keys [
           {arriveTangent=5.2354, interpMode=RCIM_Cubic, leaveTangent=5.2354, tangentMode=RCTM_User, time=-0.0046, value=0.25}
           <at 0.1439 1.4923>
           <at 1.0011 3.4809 RCIM_Cubic>
          ]}}}))
     1=&Vector2DFromFloat(Value=&FloatFromCurve(FloatCurve=DataInterface() {propertyValues {Curve {keys [<at 0 1 RCIM_Cubic>]}}}))
     2=&Vector2DFromFloat(Value=&FloatFromCurve(FloatCurve=DataInterface() {propertyValues {Curve {keys [<at -0.0046 0>,<at 0.1746 0.9952>,<at 1.0011 0.002>]}}}))
     3=?
     4=?
     Index=&SpawnGroup()
     Num=::NewEnumerator1
    )
    Scale Sprite Size Mode=::NewEnumerator2
   )
   /SetVariables(
    Particles.DynParSmoke0:Vector4f=&MakeVector4(
     X=&FloatFromCurve(FloatCurve=DataInterface() {propertyValues {Curve {keys [<at 0 0 RCIM_Cubic>,<at 1.0011 0.2>]}}})
     Y=0
     Z=&FloatFromCurve(FloatCurve=DataInterface() {propertyValues {Curve {keys [{interpMode=RCIM_Cubic, leaveTangent=4.6779, time=0.2851, value=0}]}}})
     W=0
    )
    Particles.DynParSmoke1:Vector4f=&MakeVector4(
     X=&FloatFromCurve(
      FloatCurve=DataInterface() {propertyValues {Curve {keys [<at 0 0>,<at 0.25 1>]}}}
      Scale Curve=&RandomRangeFloat(Minimum=2, Maximum=1)
     )
     Y=&FloatFromCurve(
      FloatCurve=DataInterface() {propertyValues {Curve {keys [
          {interpMode=RCIM_Cubic, leaveTangent=3.7422, tangentMode=RCTM_Break, time=0, value=0.2}
          {interpMode=RCIM_Cubic, leaveTangent=0.9066, time=1, value=2}
         ]}}}
      Scale Curve=&RandomRangeFloat(Minimum=0.5, Maximum=1)
     )
     Z=&FloatFromCurve(FloatCurve=DataInterface() {propertyValues {Curve {keys [
          {interpMode=RCIM_Cubic, leaveTangent=2.8445, tangentMode=RCTM_Break, time=0, value=0.2}
          {interpMode=RCIM_Cubic, leaveTangent=63.2978, time=0.25, value=1}
         ]}}})
     W=&FloatFromCurve(
      FloatCurve=DataInterface() {propertyValues {Curve {keys [<at 0 1>,<at 1 0.8>]}}}
      Scale Curve=&RandomRangeFloat(Minimum=0.5, Maximum=1)
     )
    )
    Particles.DynParSmoke2:Vector4f=&MakeVector4(
     X=&FloatFromCurve(FloatCurve=DataInterface() {propertyValues {Curve {keys [{interpMode=RCIM_Cubic, leaveTangent=3.2309, tangentMode=RCTM_Break, time=0.1, value=0}, <at 1 1>]}}})
     Y=$Particles.MaterialRandom
     Z=&FloatFromCurve(FloatCurve=DataInterface() {propertyValues {Curve {keys [
          {interpMode=RCIM_Cubic, tangentMode=RCTM_Break, time=0, value=1}
          {interpMode=RCIM_Cubic, leaveTangent=11.1233, tangentMode=RCTM_Break, time=1, value=2}
         ]}}})
     W=0.75
    )
   ) {isSetParameters=true}
   &ScaleColor(
    Scale Alpha=&MergeHelper_FloatSwitch(
     0=&FloatFromCurve(FloatCurve=DataInterface() {propertyValues {Curve {keys [<at 0 0 RCIM_Cubic>,<at 0.2049 1 RCIM_Cubic>,<at 1 0 RCIM_Cubic>]}}})
     1=&FloatFromCurve(FloatCurve=DataInterface() {propertyValues {Curve {keys [<at 0 0>,<at 0.25 0.75>,<at 1 0>]}}})
     2=&FloatFromCurve(
      FloatCurve=DataInterface() {propertyValues {Curve {keys [<at -0.0092 0.0144>,<at 0.0967 0.9767>,<at 1.0011 0.002>]}}}
      Scale Curve=&FN_FrontEndFloatModifier(Gameplay=1, Front End=0.5)
     )
     3=?
     Index=&SpawnGroup()
     Num=::NewEnumerator1
    )
    ScaleA=true
    ScaleRGB=true
   ) {module=ScaleColor001}
  ]
  rendererClasses [&NiagaraSpriteRendererProperties]
  renderers [
   {rendererClass=&NiagaraSpriteRendererProperties}
   {rendererClass=&NiagaraSpriteRendererProperties, rendererIndex=1}
   {rendererClass=&NiagaraSpriteRendererProperties, rendererIndex=2}
   {rendererClass=&NiagaraSpriteRendererProperties, rendererIndex=3}
  ]
 }
 Flare {
  emitterUpdateScript [
   &EmitterState()
   &SpawnBurst_Instantaneous(Use Spawn Probability=false, UseLoopCountLimit=false)
  ]
  particleSpawnScript [
   &InitializeParticle(
    Color=<rgba 0.5 1.1261 5 1>
    Lifetime=0.4
    Position Offset=<0 0 20>
    Sprite Rotation Mode=::NewEnumerator2
    Sprite Size Mode=::NewEnumerator3
    Uniform Sprite Size=170
    UsePositionOffset=true
    Write Material Random=true
   )
   &DynamicMaterialParameters(
    Index 0 Param 1=&RandomRangeFloat(Minimum=0.4, Maximum=0.6)
    Index 0 Param 2=1
    Index 0 Param 3=1
    Index 0 Param 4=1
    Module.Index0Float0Write=true
    Module.Index0Float1Write=true
    Module.Index0Float2Write=true
    Module.Index0Float3Write=true
   )
   /SetVariables(Particles.LightSpecularScale:NiagaraFloat=&FN_FrontEndFloatModifier(Gameplay=1, Front End=0)) {isSetParameters=true}
  ]
  particleUpdateScript [
   &ParticleState()
   &SolveForcesAndVelocity(Clamp Velocity=false, Limit Acceleration=false)
   &ScaleColor(
    Scale Alpha=&FloatFromCurve(FloatCurve=DataInterface() {propertyValues {Curve {keys [<at 0.8 1>,<at 1 0>]}}}, Scale Curve=$System.TODScalar)
    ScaleA=true
    ScaleRGB=true
   )
   &ScaleSpriteSize(
    Scale Factor=&Vector2DFromFloat(Value=&FloatFromCurve(FloatCurve=DataInterface() {propertyValues {Curve {keys [<at 0 1.5 RCIM_Cubic>,<at 1 0>]}}}))
    Scale Sprite Size Mode=::NewEnumerator2
   )
   &FN_Light_Attributes(
    Color=<rgba 0 3.8263 20 10>
    Create Color Attribute=true
    Create Exponent Attribute=true
    Create Position Attribute=true
    Create Radius Attribute=true
    Exponent=5
    Position=<0 0 20>
    Radius=70
    Scale Color=&FloatFromCurve(FloatCurve=DataInterface() {propertyValues {Curve {keys [<at 0.8 1>,<at 1 0>]}}}, Scale Curve=$System.TODScalar)
   )
  ]
  rendererClasses [&NiagaraSpriteRendererProperties, &NiagaraLightRendererProperties]
  renderers [
   {rendererClass=&NiagaraSpriteRendererProperties}
   {rendererClass=&NiagaraLightRendererProperties, rendererIndex=1}
  ]
 }
 Shockwave {
  emitterUpdateScript [
   &EmitterState()
   &SpawnBurst_Instantaneous(Use Spawn Probability=false, UseLoopCountLimit=false)
  ]
  particleSpawnScript [
   &InitializeParticle(
    Color=<rgba 1 1 1 5>
    Lifetime=0.3
    Mesh Scale Mode=::NewEnumerator3
    Mesh Uniform Scale=&Multiply_Float(A=0.3, B=$System.FrontendScale)
    Position Offset=<0 0 20>
    UsePositionOffset=true
    Write Material Random=true
   )
   &InitialMeshOrientation(
    Mesh Orientation Mode=::NewEnumerator7
    Rotation=&RandomRangeVector(Minimum=<0 0 0>, Maximum=<1 0 0>)
    Use Rotation Vector=true
   )
  ]
  particleUpdateScript [
   &ParticleState()
   &DynamicMaterialParameters(
    Index 0 Param 1=1
    Index 0 Param 2=2
    Module.Index0Float0Write=true
    Module.Index0Float1Write=true
    Module.Index0Float2Write=true
    Module.Index0Float3Write=true
   )
   &ScaleColor(
    Scale Alpha=&FloatFromCurve(
     FloatCurve=DataInterface() {propertyValues {Curve {keys [
         {arriveTangent=-1.5846, interpMode=RCIM_Cubic, leaveTangent=-1.5846, tangentMode=RCTM_User, time=0, value=1}
         {arriveTangent=-0.3296, tangentMode=RCTM_User, time=1, value=0}
        ]}}}
     Scale Curve=&Multiply_Float(A=0.4, B=$System.TODScalar)
    )
    Scale RGB=&MakeVectorFromLinearColorRGB(Color=&ColorFromCurve(ColorCurve=DataInterface() {propertyValues {
        AlphaCurve {keys [<at 0 1.5>,<at 1 0>]}
        BlueCurve {keys [<at 0.0389 1>,<at 1 1>]}
        GreenCurve {keys [<at 0.0389 0.4767>,<at 1 0.03>]}
        RedCurve {keys [<at 0.0389 0.0598>,<at 1 0.03>]}
       }}))
    ScaleA=true
    ScaleRGB=true
   )
   &ScaleMeshSize(Scale Factor=&VectorFromFloat(Value=&FloatFromCurve(FloatCurve=DataInterface() {propertyValues {Curve {keys [
          {arriveTangent=2.4306, interpMode=RCIM_Cubic, leaveTangent=2.4306, tangentMode=RCTM_User, time=0, value=0}
          {arriveTangent=0.046, interpMode=RCIM_Cubic, leaveTangent=0.046, tangentMode=RCTM_User, time=1, value=1}
         ]}}})))
   &CameraOffset(Camera Offset Amount=15)
   /SetVariables(
    Particles.SecondaryDyn:Vector4f=&MakeVector4(X=2.5, Y=3, Z=0, W=0)
    Particles.SecondarySphereColor:LinearColor=&MakeLinearColorFromVectorAndFloat(
     "Vector (RGB)"=&MakeVectorFromLinearColorRGB(Color=&ColorFromCurve(ColorCurve=DataInterface() {propertyValues {
         AlphaCurve {keys [<at 0 1.5>,<at 1 0>]}
         BlueCurve {keys [<at 0.0389 5>,<at 1 1>]}
         GreenCurve {keys [<at 0.0389 0.5>,<at 1 0.1>]}
         RedCurve {keys [<at 0.0389 0.5>,<at 1 0.1>]}
        }}))
     "Float (Alpha)"=&FloatFromCurve(
      FloatCurve=DataInterface() {propertyValues {Curve {keys [
          {arriveTangent=-1.5846, interpMode=RCIM_Cubic, leaveTangent=-1.5846, tangentMode=RCTM_User, time=0, value=1}
          {arriveTangent=-0.3296, tangentMode=RCTM_User, time=1, value=0}
         ]}}}
      Scale Curve=&Multiply_Float(A=3, B=$System.TODScalar)
     )
    )
    Particles.ShockwaveColor:LinearColor=&MakeLinearColorFromVectorAndFloat(
     "Vector (RGB)"=&MakeVectorFromLinearColorRGB(Color=<rgba 0.02 0.02 1 1>)
     "Float (Alpha)"=&FloatFromCurve(
      FloatCurve=DataInterface() {propertyValues {Curve {keys [
          {arriveTangent=0.0068, interpMode=RCIM_Cubic, leaveTangent=0.0068, tangentMode=RCTM_User, time=0, value=0}
          {arriveTangent=-0.018, interpMode=RCIM_Cubic, leaveTangent=-0.018, tangentMode=RCTM_User, time=0.1, value=1}
          {arriveTangent=0.0005, interpMode=RCIM_Cubic, leaveTangent=0.0005, tangentMode=RCTM_User, time=0.7, value=0}
         ]}}}
      Scale Curve=&Multiply_Float(A=0.5, B=$System.TODScalar)
     )
    )
    Particles.ShockwaveScale:Vector3f=&Multiply_VectorByFloat(Vector=$Particles.Scale, Float=-3.345)
   ) {isSetParameters=true}
  ]
  rendererClasses [&NiagaraMeshRendererProperties]
  renderers [
   {rendererClass=&NiagaraMeshRendererProperties, rendererIndex=0}
   {rendererClass=&NiagaraMeshRendererProperties, rendererIndex=1}
   {rendererClass=&NiagaraMeshRendererProperties, rendererIndex=2}
  ]
 }
}
summary {
 systemName=NS_Emote_PixiePart_DefaultPropSpawn
 emitters [
  {emitterName=DefaultPropSpawn_Merged, rendererClasses [&NiagaraSpriteRendererProperties]}
  {emitterName=Flare, rendererClasses [&NiagaraSpriteRendererProperties, &NiagaraLightRendererProperties]}
  {emitterName=BigSparkles, rendererClasses [&NiagaraSpriteRendererProperties]}
  {emitterName=Shockwave, rendererClasses [&NiagaraMeshRendererProperties]}
 ]
 userVariables [
  User.bIsFrontEnd {type=NiagaraBool}
  User.bIsFrontEndPreview {type=NiagaraBool}
 ]
}
